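# Post-process style AnimBlueprint for the female medium Quarrel Camo cosmetic body, adding deformation and secondary motion to its attached parent's copied pose.
# Twist and sculpter Control Rigs run first, then a weighted multi-bone constraint anchors the chain master to torso and neck bones.
# AnimDynamics nodes simulate the chain with a spherical limit, sprung collars and bags.
# Fixed additive offsets adjust wrist and strap bones; the event graph is empty.
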AnimBlueprint /BRCosmetics/Characters/Player/Female/Medium/Bodies/F_MED_Quarrel_Camo/Meshes/F_MED_Quarrel_Camo_AnimBP
target_skeleton=&/F_MED_Quarrel_Camo_Skeleton
graphs [
 AnimGraph {nodes [
   n1 = Copy Pose From Mesh() {bUseAttachedParent=true}
   n2 = Control Rig(Source=Component To Local(ComponentPose=Local To Component(LocalPose=@n1))) {
    alphaBoolBlend {blendOption=Linear}
    controlRigAssetReference {blueprintRigClass=&/F_MED_Quarrel_Camo_Deform_CtrlRig_twists.F_MED_Quarrel_Camo_Deform_CtrlRig_twists_C}
   }
   n3 = Control Rig(Source=@n2) {
    alphaBoolBlend {blendOption=Linear}
    controlRigAssetReference {blueprintRigClass=&/F_MED_Quarrel_Camo_Deform_CtrlRig_sculpters.F_MED_Quarrel_Camo_Deform_CtrlRig_sculpters_C}
   }
   n4 = "Constraint - dyn_chain_MSTR (PPPPPPPP)"!(
    ComponentPose=Local To Component(LocalPose=@n3)
    Alpha=1
    ConstraintWeights_0=0.0241
    ConstraintWeights_1:float=0.0004
    ConstraintWeights_2:float=0.0008
    ConstraintWeights_3:float=0.1507
    ConstraintWeights_4:float=0.7894
    ConstraintWeights_5:float=0.0218
    ConstraintWeights_6:float=0.0079
    ConstraintWeights_7:float=0.0049
   ) -> (Pose) {
    alphaBoolBlend {blendOption=Linear}
    boneToModify {boneName=dyn_chain_MSTR}
    constraintSetup [
     {targetBone {boneName=clavicle_l}}
     {targetBone {boneName=deform_pectoral_fr_r}}
     {targetBone {boneName=spine_03}}
     {targetBone {boneName=spine_04}}
     {targetBone {boneName=spine_05}}
     {targetBone {boneName=clavicle_r}}
     {targetBone {boneName=neck_02}}
     {targetBone {boneName=neck_01}}
    ]
    constraintWeights [1,1,1,1,1,1,1,1]
   }
   n5 = Anim Dynamics - Bone: dyn_chain(ComponentPose=@n4, Alpha=0.5, ExternalForce="0,0,0") {
    angularDampingOverride=0.9999
    bGravityOverrideInSimSpace=true
    bOverrideAngularDamping=true
    simulationSpace=RootRelative
    alphaBoolBlend {blendOption=Linear}
    boundBone {boneName=dyn_chain}
    physicsBodyDefinitions [{
      sphereCollisionRadius=0
      boundBone {boneName=dyn_chain}
      boxExtents=<3 10 10>
      constraintSetup {angularLimitsMax=<2 0 45>, angularLimitsMin=<-2 -5 -45>, angularTarget=<0 0 0>}
      localJointOffset=<1.5 -0.05 0.1>
     }]
    sphericalLimits [{limitRadius=20, drivingBone {boneName=spine_05}, sphereLocalOffset=<-20 -10 0>}]
   }
   n6 = Anim Dynamics - Bone: dyn_r_collar_01(ComponentPose=@n5, Alpha=0.5, ExternalForce="0,0,0") {
    angularDampingOverride=1
    angularSpringConstant=3000
    bAngularSpring=true
    bGravityOverrideInSimSpace=true
    bOverrideAngularDamping=true
    gravityScale=0.35
    simulationSpace=RootRelative
    alphaBoolBlend {blendOption=Linear}
    boundBone {boneName=dyn_r_collar_01}
    componentAppliedLinearAccClamp=<2500 2500 2500>
    componentLinearAccScale=<1 1 1>
    componentLinearVelScale=<1 1 1>
    physicsBodyDefinitions [{
      sphereCollisionRadius=0
      boundBone {boneName=dyn_r_collar_01}
      boxExtents=<5 10 10>
      constraintSetup {angularLimitsMax=<0 2 15>, angularLimitsMin=<0 -2 -15>, angularTarget=<1 0 0>}
      localJointOffset=<0.5 0 0>
     }]
   }
   n7 = Anim Dynamics - Bone: dyn_r_collar_02(ComponentPose=@n6, Alpha=0.3, ExternalForce="0,0,0") {
    angularDampingOverride=1
    angularSpringConstant=3000
    bAngularSpring=true
    bGravityOverrideInSimSpace=true
    bOverrideAngularDamping=true
    gravityScale=0.35
    simulationSpace=RootRelative
    alphaBoolBlend {blendOption=Linear}
    boundBone {boneName=dyn_r_collar_02}
    componentAppliedLinearAccClamp=<2500 2500 2500>
    componentLinearAccScale=<1 1 1>
    componentLinearVelScale=<1 1 1>
    physicsBodyDefinitions [{
      sphereCollisionRadius=0
      boundBone {boneName=dyn_r_collar_02}
      boxExtents=<5 10 10>
      constraintSetup {angularLimitsMax=<0 2 15>, angularLimitsMin=<0 -2 -15>, angularTarget=<1 0 0>}
      localJointOffset=<0.5 0 0>
     }]
   }
   n8 = Anim Dynamics - Bone: dyn_r_collar_03(ComponentPose=@n7, Alpha=0.1, ExternalForce="0,0,0") {
    angularDampingOverride=1
    angularSpringConstant=3000
    bAngularSpring=true
    bGravityOverrideInSimSpace=true
    bOverrideAngularDamping=true
    gravityScale=0.35
    simulationSpace=RootRelative
    alphaBoolBlend {blendOption=Linear}
    boundBone {boneName=dyn_r_collar_03}
    componentAppliedLinearAccClamp=<2500 2500 2500>
    componentLinearAccScale=<1 1 1>
    componentLinearVelScale=<1 1 1>
    physicsBodyDefinitions [{
      sphereCollisionRadius=0
      boundBone {boneName=dyn_r_collar_03}
      boxExtents=<5 10 10>
      constraintSetup {angularLimitsMax=<0 2 15>, angularLimitsMin=<0 -2 -15>, angularTarget=<1 0 0>}
      localJointOffset=<0.5 0 0>
     }]
   }
   n9 = Anim Dynamics - Bone: dyn_r_collar_04(ComponentPose=@n8, Alpha=0.1, ExternalForce="0,0,0") {
    angularDampingOverride=1
    angularSpringConstant=3000
    bAngularSpring=true
    bGravityOverrideInSimSpace=true
    bOverrideAngularDamping=true
    gravityScale=0.35
    simulationSpace=RootRelative
    alphaBoolBlend {blendOption=Linear}
    boundBone {boneName=dyn_r_collar_04}
    componentAppliedLinearAccClamp=<2500 2500 2500>
    componentLinearAccScale=<1 1 1>
    componentLinearVelScale=<1 1 1>
    physicsBodyDefinitions [{
      sphereCollisionRadius=0
      boundBone {boneName=dyn_r_collar_04}
      boxExtents=<5 10 10>
      constraintSetup {angularLimitsMax=<0 2 15>, angularLimitsMin=<0 -2 -15>, angularTarget=<1 0 0>}
      localJointOffset=<0.5 0 0>
     }]
   }
   n10 = Anim Dynamics - Bone: dyn_l_collar_01(ComponentPose=@n9, Alpha=0.5, ExternalForce="0,0,0") {
    angularDampingOverride=1
    angularSpringConstant=3000
    bAngularSpring=true
    bGravityOverrideInSimSpace=true
    bOverrideAngularDamping=true
    gravityScale=0.35
    simulationSpace=RootRelative
    alphaBoolBlend {blendOption=Linear}
    boundBone {boneName=dyn_l_collar_01}
    componentAppliedLinearAccClamp=<2500 2500 2500>
    componentLinearAccScale=<1 1 1>
    componentLinearVelScale=<1 1 1>
    physicsBodyDefinitions [{
      sphereCollisionRadius=0
      boundBone {boneName=dyn_l_collar_01}
      boxExtents=<5 10 10>
      constraintSetup {angularLimitsMax=<0 2 15>, angularLimitsMin=<0 -2 -15>, angularTarget=<1 0 0>}
      localJointOffset=<-0.5 0 0>
     }]
   }
   n11 = Anim Dynamics - Bone: dyn_l_collar_02(ComponentPose=@n10, Alpha=0.3, ExternalForce="0,0,0") {
    angularDampingOverride=1
    angularSpringConstant=3000
    bAngularSpring=true
    bGravityOverrideInSimSpace=true
    bOverrideAngularDamping=true
    gravityScale=0.35
    simulationSpace=RootRelative
    alphaBoolBlend {blendOption=Linear}
    boundBone {boneName=dyn_l_collar_02}
    componentAppliedLinearAccClamp=<2500 2500 2500>
    componentLinearAccScale=<1 1 1>
    componentLinearVelScale=<1 1 1>
    physicsBodyDefinitions [{
      sphereCollisionRadius=0
      boundBone {boneName=dyn_l_collar_02}
      boxExtents=<5 10 10>
      constraintSetup {angularLimitsMax=<0 -1 15>, angularLimitsMin=<0 -10 -15>, angularTarget=<1 0 0>}
      localJointOffset=<-0.5 0 0>
     }]
   }
   n12 = Anim Dynamics - Bone: dyn_l_collar_03(ComponentPose=@n11, Alpha=0.1, ExternalForce="0,0,0") {
    angularDampingOverride=1
    angularSpringConstant=3000
    bAngularSpring=true
    bGravityOverrideInSimSpace=true
    bOverrideAngularDamping=true
    gravityScale=0.35
    simulationSpace=RootRelative
    alphaBoolBlend {blendOption=Linear}
    boundBone {boneName=dyn_l_collar_03}
    componentAppliedLinearAccClamp=<2500 2500 2500>
    componentLinearAccScale=<1 1 1>
    componentLinearVelScale=<1 1 1>
    physicsBodyDefinitions [{
      sphereCollisionRadius=0
      boundBone {boneName=dyn_l_collar_03}
      boxExtents=<5 10 10>
      constraintSetup {angularLimitsMax=<0 2 15>, angularLimitsMin=<0 -2 -15>, angularTarget=<1 0 0>}
      localJointOffset=<-0.5 0 0>
     }]
   }
   n13 = Anim Dynamics - Bone: dyn_l_collar_04(ComponentPose=@n12, Alpha=0.1, ExternalForce="0,0,0") {
    angularDampingOverride=1
    angularSpringConstant=3000
    bAngularSpring=true
    bGravityOverrideInSimSpace=true
    bOverrideAngularDamping=true
    gravityScale=0.35
    simulationSpace=RootRelative
    alphaBoolBlend {blendOption=Linear}
    boundBone {boneName=dyn_l_collar_04}
    componentAppliedLinearAccClamp=<2500 2500 2500>
    componentLinearAccScale=<1 1 1>
    componentLinearVelScale=<1 1 1>
    physicsBodyDefinitions [{
      sphereCollisionRadius=0
      boundBone {boneName=dyn_l_collar_04}
      boxExtents=<5 10 10>
      constraintSetup {angularLimitsMax=<0 2 15>, angularLimitsMin=<0 -2 -15>, angularTarget=<1 0 0>}
      localJointOffset=<-0.5 0 0>
     }]
   }
   n14 = Anim Dynamics - Bone: dyn_bag_01(ComponentPose=@n13, Alpha=0.9, ExternalForce="0,0,0") {
    angularDampingOverride=0.997
    angularSpringConstant=60
    bAngularSpring=true
    bGravityOverrideInSimSpace=true
    bOverrideAngularDamping=true
    gravityScale=0.8
    simulationSpace=RootRelative
    alphaBoolBlend {blendOption=Linear}
    boundBone {boneName=dyn_bag_01}
    physicsBodyDefinitions [{
      sphereCollisionRadius=0
      boundBone {boneName=dyn_bag_01}
      boxExtents=<15 22 22>
      constraintSetup {angularLimitsMax=<20 10 5>, angularLimitsMin=<-20 -20 -5>, angularTarget=<0 0 0>}
      localJointOffset=<1.5 -0.05 0.1>
     }]
   }
   n15 = Anim Dynamics - Bone: dyn_bag_02(ComponentPose=@n14, Alpha=0.9, ExternalForce="0,0,0") {
    angularDampingOverride=0.9999
    angularSpringConstant=60
    bAngularSpring=true
    bGravityOverrideInSimSpace=true
    bOverrideAngularDamping=true
    gravityScale=0.8
    simulationSpace=RootRelative
    alphaBoolBlend {blendOption=Linear}
    boundBone {boneName=dyn_bag_02}
    physicsBodyDefinitions [{
      sphereCollisionRadius=0
      boundBone {boneName=dyn_bag_02}
      boxExtents=<15 22 22>
      constraintSetup {angularLimitsMax=<20 10 5>, angularLimitsMin=<-20 -20 -5>, angularTarget=<0 0 0>}
      localJointOffset=<1.5 -0.05 0.1>
     }]
   }
   n16 = "Transform (Modify) Bone - Bone: dyn_l_wristFix"(ComponentPose=@n15, Alpha=1, Translation="0,1.2,0", Rotation="0,0,0", Scale="1,1,1") {
    rotationMode=BMM_Additive
    rotationSpace=BCS_BoneSpace
    scaleMode=BMM_Additive
    scaleSpace=BCS_BoneSpace
    translationMode=BMM_Additive
    translationSpace=BCS_BoneSpace
    alphaBoolBlend {blendOption=Linear}
    boneToModify {boneName=dyn_l_wristFix}
    translation {y=1.2}
   }
   n17 = "Transform (Modify) Bone - Bone: dyn_r_wristFix"(ComponentPose=@n16, Alpha=1, Translation="1.5,-3,0", Rotation="0,0,0", Scale="1,1,1") {
    rotationMode=BMM_Additive
    rotationSpace=BCS_BoneSpace
    scaleMode=BMM_Additive
    scaleSpace=BCS_BoneSpace
    translationMode=BMM_Additive
    translationSpace=BCS_BoneSpace
    alphaBoolBlend {blendOption=Linear}
    boneToModify {boneName=dyn_r_wristFix}
    translation=<1.5 -3>
   }
   n18 = "Transform (Modify) Bone - Bone: dyn_Strap_Expand"(ComponentPose=@n17, Alpha=1, Translation="0,-1.2,0", Rotation="0,0,0", Scale="1,1,1") {
    rotationMode=BMM_Additive
    rotationSpace=BCS_BoneSpace
    scaleMode=BMM_Additive
    scaleSpace=BCS_BoneSpace
    translationMode=BMM_Additive
    translationSpace=BCS_BoneSpace
    alphaBoolBlend {blendOption=Linear}
    boneToModify {boneName=dyn_Strap_Expand}
    translation {y=-1.2}
   }
   Output Pose(Result=Component To Local(ComponentPose=@n18)) {layerGroup=None}
  ]}
 EventGraph {nodes [Event BlueprintUpdateAnimation()]}
]
variables []
